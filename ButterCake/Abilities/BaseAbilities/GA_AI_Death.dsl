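; AI death gameplay ability that plays a death montage and hides the pawn's body when it ends.
; On activation it commits, blocks cancellation, and checks the montage from MontageInfo_MontageToPlay.
; If valid, it runs PlayMontageInfoAndWaitForTrigger with the separate MontageInfo variable before ending; otherwise it ends immediately.
; OnEndAbility calls HideBodyonDeath on the activating pawn, if valid, regardless of cancellation.
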(event Ability|EventActivateAbility
  (bind _montageinfo_montagetoplay (Utilities|Struct|BreakFortGameplayAbilityMontageInfo (Variables|GetMontageInfo_MontageToPlay)))
  (Ability|CommitAbility)
  (Ability|SetCanbeCanceled false)
  (Utilities|IsValid _montageinfo_montagetoplay
    (:Is%20Valid
      (Ability|Tasks|PlayMontageInfoAndWaitForTrigger "None" (Variables|Animation|GetMontageInfo)
        (:Cancelled
          _)
        (:Completed
          _))
      (Ability|EndAbility))
    (:Is%20Not%20Valid
      (Ability|EndAbility))))

(event Ability|EventOnEndAbility
  (params (bWasCancelled :type Boolean))
  (bind _returnvalue (Ability|GetActivatingPawn))
  (Utilities|IsValid _returnvalue
    (:Is%20Valid
      (FortPawn|HideBodyonDeath _returnvalue false))))
