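; Base AI gameplay ability that moves its avatar to a location chosen by a configurable EQS query.
; On activation it commits and runs the query, then passes the first result location to a MoveAI ability task with configurable stop and nav filter settings.
; It ends when the move finishes or setup fails, and on end unbinds and clears the stored task.
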(event Ability|EventActivateAbility
  (Ability|CommitAbility)
  (bind _moveeqs (Variables|MoveToEQS|GetMoveEQS)
    (:then
      (bind _returnvalue (AI|EQS|RunEQSQuery _moveeqs (Ability|GetAvatarActorfromActorInfo) "AllMatching"))
      (AI|EQS|BindEventtoOnQueryFinished _returnvalue (delegate Custom|OnMoveQueryFinishedEvent)))
    (:else
      (Ability|EndAbility))))

(event Custom|OnMoveQueryFinishedEvent
  (params (QueryInstance :type EQS%20Query%20Instance%20Object%20Reference) (QueryStatus :type EEnvQueryStatus%20Enum :default "Processing"))
  (if (Utilities|Enum|Equal(Enum) QueryStatus "Success")
    (bind _resultlocations (AI|EQS|GetQueryResultsasLocations QueryInstance))
    (if (Utilities|Array|IsValidIndex _resultlocations)
      (bind _asynctaskproxy (Ability|Tasks|MoveAI "MoveToEQS" 0 (Utilities|Array|Get(acopy) _resultlocations) "UsePathfinding" "Ignore" (Variables|MovetoEQS|GetMoveAI_StopAtEnd) "Default" false true -1.0 0 (Variables|MovetoEQS|GetNavFilterClass) "(TagName=\"\")"))
      (Utilities|IsValid _asynctaskproxy
        (:Is%20Valid
          (bind _output_get (Variables|MovetoEQS|SetMoveTask _asynctaskproxy))
          (EventDispatchers|BindEventtoOnMoveFinished _output_get (delegate Custom|OnMoveFinishedEvent)))
        (:Is%20Not%20Valid
          _))
      (else
        _))
    (Ability|EndAbility)
    (else
      (Ability|EndAbility))))

(event Custom|OnMoveFinishedEvent
  (params (Result :type EPathFollowingResult%20Enum :default "Success") (AIController :type AIController%20Object%20Reference))
  (Ability|EndAbility))

(event Ability|EventOnEndAbility
  (params (bWasCancelled :type Boolean))
  (bind _movetask (Variables|MovetoEQS|GetMoveTask)
    (:then
      (EventDispatchers|UnbindallEventsfromOnMoveFinished _movetask)
      (Variables|MovetoEQS|SetMoveTask 0))
    (:else)))
