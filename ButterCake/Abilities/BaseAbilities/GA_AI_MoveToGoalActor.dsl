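; Gameplay ability that moves an AI character to the actor returned as its AI ability target.
; On activation it commits, starts a pathfinding MoveToGoalActor AI task with configurable stop-at-end and nav filter, and stores that task.
; An invalid target or task ends the ability immediately; otherwise the task's move-finished event ends it, whatever the result.
; Ending the ability unbinds that event and clears the stored task.
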(event Ability|EventActivateAbility
  (bind _returnvalue (Ability|GetAIAbilityTarget))
  (Ability|CommitAbility)
  (Utilities|IsValid _returnvalue
    (:Is%20Valid
      (bind _asynctaskproxy (Ability|Tasks|MoveAI "MoveToGoalActor" _returnvalue "0, 0, 0" "UsePathfinding" "Ignore" (Variables|MovetoGoalActor|GetMoveAI_StopAtEnd) "Default" false true -1.0 0 (Variables|MoveToGoalActor|GetNavFilterClass) "(TagName=\"\")"))
      (Utilities|IsValid _asynctaskproxy
        (:Is%20Valid
          (bind _output_get (Variables|MovetoGoalActor|SetMoveTask _asynctaskproxy))
          (EventDispatchers|BindEventtoOnMoveFinished _output_get (delegate Custom|OnMoveFinishedEvent)))
        (:Is%20Not%20Valid
          _)))
    (:Is%20Not%20Valid
      _))
  (Ability|EndAbility))

(event Custom|OnMoveFinishedEvent
  (params (Result :type EPathFollowingResult%20Enum :default "Success") (AIController :type AIController%20Object%20Reference))
  (Ability|EndAbility))

(event Ability|EventOnEndAbility
  (params (bWasCancelled :type Boolean))
  (bind _movetask (Variables|MovetoGoalActor|GetMoveTask)
    (:then
      (EventDispatchers|UnbindallEventsfromOnMoveFinished _movetask)
      (Variables|MovetoGoalActor|SetMoveTask 0))
    (:else)))
